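# Parent material for the CanyonVale character, layering a crystal gem effect and shirt tint between the shared 2023 character PreFX and PostFX functions.
# Tech Art Mask channels declare gem (R) and shirt (B) masks, with ShirtColor multiplying masked base color.
# MF_Crystal's blend is scaled by hasGem and dimmed toward GemShadowIntensity on surfaces facing away from the anime shader's light direction, using a pixel-normal dot product.
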Material /BRCosmetics/Characters/Player/Male/Medium/Bodies/M_MED_CanyonVale/Materials/M_CanyonValeCharacter_Parent
graph {
 properties {
  bAutomaticallySetUsageInEditor=false
  bUseMaterialAttributes=true
  bUsedWithMeshDeformer=true
  bUsedWithMorphTargets=true
  bUsedWithSkeletalMesh=true
  pixelDepthOffsetMode=PDOM_Legacy
  previewMesh=/BRCosmetics/Accessories/FORT_Backpacks/Backpack_M_MED_CanyonVale/Meshes/M_MED_CanyonVale_Pack
  shadingModel=MSM_FromMaterialExpression
  usedShadingModels=Default Lit | Subsurface | Clear Coat | Subsurface Profile | Cloth
 }
 nodes [
  n1 = MaterialFunctionCall() {materialFunction=&MF_CharacterParent_2023_PreFX}
  n2 = GetMaterialAttributes(None=@n1.Result) {attributeGetTypes [#1, #2]}
  n3 = ScalarParameter() {group=Crystal, parameterName=hasGem, sliderMax=1}
  n4 = ScalarParameter() {defaultValue=0.35, group=Crystal, parameterName=GemShadowIntensity, sliderMax=1}
  n5 = Multiply(A=DotProduct(
    A=PixelNormalWS()
    B=MaterialFunctionCall() {materialFunction=/Game/Global/Materials/Character_Parent_2023/Anime_Parent_2023/MaterialFunctions/MFC_AnimeShader_LightDirection}.Direction Vector
   )) {constB=2}
  n6 = LinearInterpolate(B=NamedRerouteUsage(), Alpha=LinearInterpolate(A=Multiply(A=@n3, B=@n4), B=@n3, Alpha=Saturate(None=@n5)) {constA=0.5})
  n7 = Multiply(A=@n2.Base Color, B=VectorParameter() {parameterName=ShirtColor, defaultValue=<rgba 0.3788 0 0.7274 1>}.RGB)
  n8 = NamedRerouteUsage()
  MaterialFunctionCall(
   Material=SetMaterialAttributes!(
    MaterialAttributes=MaterialFunctionCall!(Material=@n2.MaterialAttributes, Alpha=@n6) {materialFunction=/Game/Global/MaterialFunctions/TechArt/MF_Crystal}.Result
    Base Color=LinearInterpolate(A=@n2.Base Color, B=@n7, Alpha=@n8)
   ) {attributeSetTypes [#1]}
   M=@n1.M
   Pre Skinned Local Position=@n1.M
   Distance Blend=@n1.M
   Fresnel=@n1.M
  ) {materialFunction=&MF_CharacterParent_2023_PostFX}
  NamedRerouteDeclaration(None=ComponentMask(None=@n1.Tech Art Mask) {b=true}) {name=ShirtMask, nodeColor=<rgba 0.1953 0 1 1>}
  NamedRerouteDeclaration(None=ComponentMask(None=@n1.Tech Art Mask) {r=true}) {name=GemMask, nodeColor=<rgba 0 1 0.5078 1>}
 ]
}
