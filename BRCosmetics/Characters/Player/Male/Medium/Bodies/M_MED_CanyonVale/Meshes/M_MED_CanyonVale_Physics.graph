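# Physics asset for the CanyonVale medium male player body, covering two right-side dynamic "can" accessory chains, likely dangling from the hip.
# Each chain runs base, 01, 02 with capsule bodies, tightly limited at the first joint and looser at the tip.
# Both bases constrain to the pelvis with free swing and twist.
# Pelvis and right thigh capsules supply body volume, and the thigh has no constraint.
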PhysicsAsset /BRCosmetics/Characters/Player/Male/Medium/Bodies/M_MED_CanyonVale/Meshes/M_MED_CanyonVale_Physics
bodies {
 dyn_can_a_01_r [{
   shapeType=Capsule
   length=8.8241
   radius=3.988
   shapeName=dyn_can_a_01_r_capsule
   center=<1.8123 -0.8034 -3.0094>
   rotation=<rot 0 89.8687 -88.3151>
  }]
 dyn_can_a_02_r [{
   shapeType=Capsule
   length=1.3087
   radius=1.3079
   shapeName=dyn_can_a_02_r_capsule
   center=<1.1205 0.1372 -0.0052>
   rotation=<rot 0 -78.7644 -89.1839>
  }]
 dyn_can_a_base_r [{shapeType=Capsule, length=0.505, radius=0.505, shapeName=dyn_can_a_base_r_capsule, rotation=<rot 0 -90 0>}]
 dyn_can_b_01_r [{
   shapeType=Capsule
   length=8.8029
   radius=4.0909
   shapeName=dyn_can_b_01_r_capsule
   center=<1.7399 -1.0186 -3.0609>
   rotation=<rot 0 89.8148 -89.7736>
  }]
 dyn_can_b_02_r [{
   shapeType=Capsule
   length=0.505
   radius=0.505
   shapeName=dyn_can_b_02_r_capsule
   center=<1.1005 0.1081 -0.0269>
   rotation=<rot 0 -78.4324 -88.0754>
  }]
 dyn_can_b_base_r [{shapeType=Capsule, length=0.505, radius=0.505, shapeName=dyn_can_b_base_r_capsule, rotation=<rot 0 -90 0>}]
 pelvis [{shapeType=Capsule, length=7.6456, radius=10.0205, center=<5.519 0 0>, rotation=<rot 0 -90 0>}]
 thigh_r [{shapeType=Capsule, length=23.3907, radius=9.0764, center=<9.8408 0 -0.8838>, rotation=<rot 0 -90 -95.2062>}]
}
constraints [
 {bone1Name=dyn_can_a_01_r, bone2Name=dyn_can_a_base_r, swing1LimitDegrees=5, swing2LimitDegrees=5, twistLimitDegrees=5}
 {bone1Name=dyn_can_a_02_r, bone2Name=dyn_can_a_01_r, swing1LimitDegrees=15, swing2LimitDegrees=40, twistLimitDegrees=10}
 {bone1Name=dyn_can_b_01_r, bone2Name=dyn_can_b_base_r, swing1LimitDegrees=5, swing2LimitDegrees=5, twistLimitDegrees=5}
 {bone1Name=dyn_can_b_02_r, bone2Name=dyn_can_b_01_r, swing1LimitDegrees=15, swing2LimitDegrees=40, twistLimitDegrees=10}
 {
  bone1Name=dyn_can_b_base_r
  bone2Name=pelvis
  swing1LimitDegrees=45
  swing1Motion=Free
  swing2LimitDegrees=45
  swing2Motion=Free
  twistLimitDegrees=45
  twistMotion=Free
 }
 {
  bone1Name=dyn_can_a_base_r
  bone2Name=pelvis
  swing1LimitDegrees=45
  swing1Motion=Free
  swing2LimitDegrees=45
  swing2Motion=Free
  twistLimitDegrees=45
  twistMotion=Free
 }
]
