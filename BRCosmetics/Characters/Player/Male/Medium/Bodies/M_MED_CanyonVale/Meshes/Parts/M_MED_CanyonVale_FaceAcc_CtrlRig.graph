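# Control Rig for the CanyonVale face accessory mesh that procedurally sways dynamic bones, including hair, according to pawn movement.
# The length of PawnVelocity is remapped to a 0-1 range and smoothed with a Kalman filter.
# That value drives the maximum wave of a single Item Harmonics node, which applies noisy oscillation with per-bone ratios instead of physics simulation.
# The PawnSpeed variable is declared but unused in the graph.
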ControlRigBlueprint /BRCosmetics/Characters/Player/Male/Medium/Bodies/M_MED_CanyonVale/Meshes/Parts/M_MED_CanyonVale_FaceAcc_CtrlRig
graphs [
 RigVMFunctionLibrary {links [], nodes []}
 RigVMModel {nodes [
   RigUnit_BeginExecution = &RigUnit_BeginExecution()
   RigVMFunction_MathVectorLength = /Script/RigVM.RigVMFunction_MathVectorLength(Value=RigVMVariableNode() {variable=PawnVelocity}.Value) {is=struct RigVMUnitNode, template="Length::Execute(in Value,out Result)"}
   Remap = &RigVMFunction_MathDoubleRemap(SourceMaximum=1400, SourceMinimum=0, TargetMaximum=1, TargetMinimum=0, Value=@RigVMFunction_MathVectorLength.Result)
   KalmanFilter = &RigVMFunction_KalmanFloat(BufferSize=8, Value=@Remap.Result) {template="KalmanFilter::Execute(in Value,in BufferSize,out Result)"}
   /Script/ControlRig.RigUnit_ItemHarmonics(
    RotationOrder=YZX
    Targets="((Item=(Type=Bone,Name=\"dyn_hair_01\"),Ratio=1),(Item=(Type=Bone,Name=\"dyn_hair_02\"),Ratio=1),(Item=(Type=Bone,Name=\"dyn_hair_03\"),Ratio=0.8),(Item=(Type=Bone,Name=\"dyn_hair_04\"),Ratio=0.9),(Item=(Type=Bone,Name=\"dyn_hair_05\"),Ratio=0.4),(Item=(Type=Bone,Na ...(truncated)"
    WaveAmplitude=(X=8,Y=24,Z=8)
    WaveEase=Linear
    WaveFrequency=(X=0.2,Y=0.4,Z=0.4)
    WaveMinimum=0
    WaveNoise=(X=0.1,Y=0.1,Z=0.1)
    WaveOffset=(X=0,Y=1,Z=2)
    WaveSpeed=(X=45,Y=45,Z=45)
    WaveMaximum=@KalmanFilter.Result
    ExecutePin=@RigUnit_BeginExecution.ExecutePin
   ) {is=struct RigVMUnitNode}
  ]}
]
hierarchy [
 root {loc=<0 0 0>, rot=<0 0 0 1>}
 pelvis {parent=root, loc=<0 -0.7071 93.1191>, rot=<0.0127 0.707 -0.0127 -0.707>}
 spine_01 {parent=pelvis, loc=<6.7494 -0.6388 0>, rot=<0 0 0.034 -0.9994>}
 spine_02 {parent=spine_01, loc=<7.2518 -0.035 0>, rot=<0 0 0.0238 -0.9997>}
 spine_03 {parent=spine_02, loc=<6.0104 0 0>, rot=<0 0 0.0149 -0.9999>}
 spine_04 {parent=spine_03, loc=<8.1024 0 0>, rot=<0 0 -0.0132 -0.9999>}
 spine_05 {parent=spine_04, loc=<10.0146 0 0>, rot=<0 0 -0.0015 -1>}
 neck_01 {parent=spine_05, loc=<11.3138 0.1529 0>, rot=<0 0 -0.2003 -0.9797>}
 neck_02 {parent=neck_01, loc=<4.5627 0 0>, rot=<0 0 0.0504 -0.9987>}
 head {parent=neck_02, loc=<5.3914 0 0>, rot=<0 0 0.115 -0.9934>}
 dyn_beard_base_l {parent=head, loc=<2.6131 -0.3332 -8.1122>, rot=<-0.6232 0.2352 0.7455 0.0238>}
 dyn_beard_01_l {parent=dyn_beard_base_l, loc=<0 0 0>, rot=<0 0 0 1>}
 dyn_beard_base_r {parent=head, loc=<2.6126 -0.3332 8.1122>, rot=<-0.2352 -0.6232 0.0238 -0.7455>}
 dyn_beard_01_r {parent=dyn_beard_base_r, loc=<0 0 0>, rot=<0 0 0 1>}
 dyn_hair_01 {parent=head, loc=<15.0326 -2.317 0.0899>, rot=<0.9431 -0.3194 0.0904 0.0196>}
 dyn_hair_02 {parent=head, loc=<15.0745 -2.3058 -1.5536>, rot=<0.9474 -0.3197 -0.0102 0.0118>}
 dyn_hair_03 {parent=head, loc=<14.4636 -1.8995 -3.1658>, rot=<0.9474 -0.3197 -0.0102 0.0118>}
 dyn_hair_04 {parent=head, loc=<14.1573 -1.6996 1.9981>, rot=<0.9474 -0.3198 -0.0066 -0.0132>}
 dyn_hair_05 {parent=head, loc=<13.7349 3.7893 1.834>, rot=<0.9203 -0.3303 -0.0582 -0.2016>}
 dyn_hair_06 {parent=head, loc=<14.0601 3.0043 -2.1522>, rot=<0.9107 -0.3888 0.0176 0.1386>}
 dyn_hair_07 {parent=head, loc=<14.109 1.4117 1.1022>, rot=<0.953 -0.2775 -0.1093 -0.053>}
 dyn_hair_08 {parent=head, loc=<13.7403 5.2206 -1.8022>, rot=<0.9406 -0.3124 -0.0173 0.132>}
 dyn_hair_09 {parent=head, loc=<13.5634 5.406 0.6438>, rot=<0.9254 -0.3633 0.0277 0.1039>}
 dyn_hair_010 {parent=head, loc=<13.5134 3.9779 -4.1445>, rot=<0.9107 -0.3888 0.0176 0.1386>}
 dyn_hair_011 {parent=head, loc=<13.3218 4.4003 3.591>, rot=<0.9203 -0.3303 -0.0582 -0.2016>}
 dyn_ear_base_l {parent=head, loc=<8.5059 -0.3232 -9.2188>, rot=<0.0377 -0.7517 0.0513 -0.6564>}
 dyn_ear_01_l {parent=dyn_ear_base_l, loc=<0 0 0>, rot=<0 0 0 1>}
 dyn_ear_02_l {parent=dyn_ear_01_l, loc=<4.7674 -0.0416 -0.072>, rot=<0.0038 0.0179 0.0187 -0.9997>}
 dyn_ear_03_l {parent=dyn_ear_02_l, loc=<3.6611 0 0>, rot=<0 0.1279 0 -0.9918>}
 dyn_ear_base_r {parent=head, loc=<8.5559 -0.4651 10.3254>, rot=<0.7631 0.0368 -0.6431 -0.0519>}
 dyn_ear_01_r {parent=dyn_ear_base_r, loc=<0 0 0>, rot=<0 0 0 -1>}
 dyn_ear_02_r {parent=dyn_ear_01_r, loc=<-4.7674 0.0416 0.0714>, rot=<0.0038 0.0179 0.0187 -0.9997>}
 dyn_ear_03_r {parent=dyn_ear_02_r, loc=<-3.6611 0 0.0005>, rot=<0 0.1279 0 -0.9918>}
 faceAttach {parent=head, loc=<0 0 0>, rot=<0.0029 -0.7071 -0.0029 -0.7071>}
 L_lip_corner {parent=faceAttach, loc=<2.7621 13.7752 -3.0267>, rot=<0.1024 -0.0116 -0.1701 -0.98>}
 R_lip_corner {parent=faceAttach, loc=<-2.7621 13.7752 -3.0267>, rot=<0.1024 0.0116 0.1701 -0.98>}
 C_jaw {parent=faceAttach, loc=<-0.0134 1.0221 3.4807>, rot=<0 0 0 1>}
 dyn_beard_base {parent=C_jaw, loc=<0.0134 11.8767 -9.2528>, rot=<-0.6766 0.134 0.7128 -0.1272>}
 dyn_beard_01 {parent=dyn_beard_base, loc=<0 0 0>, rot=<0 0 0 1>}
 dyn_beard_02 {parent=dyn_beard_01, loc=<2.3162 0 0>, rot=<-0.0199 -0.069 0.2759 -0.9585>}
 dyn_beard_03 {parent=dyn_beard_02, loc=<1.7223 0 0>, rot=<0 0 0 1>}
 L_lip_lower_outer {parent=C_jaw, loc=<1.8258 13.8778 -7.331>, rot=<0.1182 0.0128 0.3286 -0.9369>}
 C_lip_lower_mid {parent=C_jaw, loc=<0.0134 14.5644 -7.51>, rot=<0.142 0 0 -0.9899>}
 R_lip_lower_outer {parent=C_jaw, loc=<-1.799 13.8778 -7.331>, rot=<0.1182 -0.0128 -0.3286 -0.9369>}
]
variables [
 PawnSpeed {type=double, default=0}
 PawnVelocity {type=FVector, default=(X=0,Y=0,Z=0)}
]
