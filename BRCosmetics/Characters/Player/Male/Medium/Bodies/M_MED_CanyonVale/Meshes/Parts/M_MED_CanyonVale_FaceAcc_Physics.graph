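# Physics asset for the M_MED_CanyonVale face accessory mesh, defining capsule bodies and joints for its dynamic beard and ear bones.
# A three-link center beard, single-link left and right beard pieces and two three-link ear chains hang from small base capsules, alongside larger head and spine_05 capsules.
# Beard bases and head link to spine_05, and ear bases to head, with free angular motion.
# Chain joints have angular limits, widening along each ear toward the tip.
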PhysicsAsset /BRCosmetics/Characters/Player/Male/Medium/Bodies/M_MED_CanyonVale/Meshes/Parts/M_MED_CanyonVale_FaceAcc_Physics
bodies {
 dyn_beard_01 [{shapeType=Capsule, length=4, radius=2.5, shapeName=dyn_beard_01_capsule, center=<2 0 0>}]
 dyn_beard_01_l [{shapeType=Capsule, length=2, radius=2, shapeName=dyn_beard_01_l_capsule, center=<1 0 0>, rotation=<rot -10.0224 -90 2>}]
 dyn_beard_01_r [{shapeType=Capsule, length=2, radius=2, shapeName=dyn_beard_01_r_capsule, center=<-1 0 0>, rotation=<rot -10.0116 -90 2>}]
 dyn_beard_02 [{shapeType=Capsule, length=4, radius=2.5, shapeName=dyn_beard_02_capsule, center=<2 0 0>}]
 dyn_beard_03 [{shapeType=Capsule, length=4, radius=2.5, shapeName=dyn_beard_03_capsule, center=<2 0 0>}]
 dyn_beard_base [{shapeType=Capsule, length=0.505, radius=0.505, shapeName=dyn_beard_base_capsule, rotation=<rot 0 -90 0>}]
 dyn_beard_base_l [{shapeType=Capsule, length=0.505, radius=0.505, shapeName=dyn_beard_base_l_capsule, rotation=<rot 0 -90 0>}]
 dyn_beard_base_r [{shapeType=Capsule, length=0.505, radius=0.505, shapeName=dyn_beard_base_r_capsule, rotation=<rot 0 -90 0>}]
 dyn_ear_01_l [{shapeType=Capsule, length=3.2923, radius=2.4805, shapeName=dyn_ear_01_l_capsule, center=<2 0 0>}]
 dyn_ear_01_r [{shapeType=Capsule, length=3.2923, radius=2.4805, shapeName=dyn_ear_01_r_capsule, center=<-2 0 0>}]
 dyn_ear_02_l [{shapeType=Capsule, length=2.6974, radius=1.8358, shapeName=dyn_ear_02_l_capsule, center=<2 0 0>}]
 dyn_ear_02_r [{shapeType=Capsule, length=2.6974, radius=1.8358, shapeName=dyn_ear_02_r_capsule, center=<-2 0 0>}]
 dyn_ear_03_l [{shapeType=Capsule, length=2.2014, radius=1.5891, shapeName=dyn_ear_03_l_capsule, center=<2 0 0>}]
 dyn_ear_03_r [{shapeType=Capsule, length=2.2014, radius=1.5891, shapeName=dyn_ear_03_r_capsule, center=<-2 0 0>}]
 dyn_ear_base_l [{shapeType=Capsule, length=0.505, radius=0.505, shapeName=dyn_ear_base_l_capsule, rotation=<rot 0 -90 0>}]
 dyn_ear_base_r [{shapeType=Capsule, length=0.505, radius=0.505, shapeName=dyn_ear_base_r_capsule, rotation=<rot 0 -90 0>}]
 head [{shapeType=Capsule, length=8.9099, radius=6.6088, center=<2.4623 4.0994 0.0165>, rotation=<rot 0 -128.737 -89.9545>}]
 spine_05 [{shapeType=Capsule, length=10.1975, radius=16.1909, shapeName=spine_05_capsule, center=<-3.5992 -2.4363 0>, rotation=<rot 0 -90 0>}]
}
constraints [
 {bone1Name=dyn_beard_01, bone2Name=dyn_beard_base, swing1LimitDegrees=24, swing2LimitDegrees=24, twistLimitDegrees=24}
 {bone1Name=dyn_beard_02, bone2Name=dyn_beard_01, swing1LimitDegrees=36, swing2LimitDegrees=45, twistLimitDegrees=36}
 {bone1Name=dyn_beard_03, bone2Name=dyn_beard_02, swing1LimitDegrees=36, swing2LimitDegrees=45, twistLimitDegrees=36}
 {bone1Name=dyn_beard_01_l, bone2Name=dyn_beard_base_l, swing1LimitDegrees=12, swing2LimitDegrees=12, twistLimitDegrees=12}
 {bone1Name=dyn_beard_01_r, bone2Name=dyn_beard_base_r, swing1LimitDegrees=12, swing2LimitDegrees=12, twistLimitDegrees=12}
 {
  bone1Name=head
  bone2Name=spine_05
  swing1LimitDegrees=45
  swing1Motion=Free
  swing2LimitDegrees=45
  swing2Motion=Free
  twistLimitDegrees=45
  twistMotion=Free
 }
 {
  bone1Name=dyn_beard_base
  bone2Name=spine_05
  swing1LimitDegrees=45
  swing1Motion=Free
  swing2LimitDegrees=45
  swing2Motion=Free
  twistLimitDegrees=45
  twistMotion=Free
 }
 {
  bone1Name=dyn_beard_base_l
  bone2Name=spine_05
  swing1LimitDegrees=45
  swing1Motion=Free
  swing2LimitDegrees=45
  swing2Motion=Free
  twistLimitDegrees=45
  twistMotion=Free
 }
 {
  bone1Name=dyn_beard_base_r
  bone2Name=spine_05
  swing1LimitDegrees=45
  swing1Motion=Free
  swing2LimitDegrees=45
  swing2Motion=Free
  twistLimitDegrees=45
  twistMotion=Free
 }
 {
  bone1Name=dyn_ear_base_l
  bone2Name=head
  swing1LimitDegrees=45
  swing1Motion=Free
  swing2LimitDegrees=45
  swing2Motion=Free
  twistLimitDegrees=45
  twistMotion=Free
 }
 {bone1Name=dyn_ear_01_l, bone2Name=dyn_ear_base_l, swing1LimitDegrees=4, swing2LimitDegrees=8, twistLimitDegrees=4}
 {bone1Name=dyn_ear_02_l, bone2Name=dyn_ear_01_l, swing1LimitDegrees=8, swing2LimitDegrees=16, twistLimitDegrees=8}
 {bone1Name=dyn_ear_03_l, bone2Name=dyn_ear_02_l, swing1LimitDegrees=16, swing2LimitDegrees=38, twistLimitDegrees=16}
 {
  bone1Name=dyn_ear_base_r
  bone2Name=head
  swing1LimitDegrees=45
  swing1Motion=Free
  swing2LimitDegrees=45
  swing2Motion=Free
  twistLimitDegrees=45
  twistMotion=Free
 }
 {bone1Name=dyn_ear_01_r, bone2Name=dyn_ear_base_r, swing1LimitDegrees=4, swing2LimitDegrees=8, twistLimitDegrees=4}
 {bone1Name=dyn_ear_02_r, bone2Name=dyn_ear_01_r, swing1LimitDegrees=8, swing2LimitDegrees=16, twistLimitDegrees=8}
 {bone1Name=dyn_ear_03_r, bone2Name=dyn_ear_02_r, swing1LimitDegrees=16, swing2LimitDegrees=38, twistLimitDegrees=16}
]
